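# Additive, unlit, two-sided magic trail material for the FishBowl pickaxe, usable on Niagara ribbons and mesh particles.
# Noise-distorted, panning plasma coils at two scales are lighten-blended.
# AorB switches between plain boosted coils and an HDR version broken up by distorted noise, tinted by particle color.
# Opacity erodes through a UV edge mask driven by a dynamic parameter, times particle alpha, with depth fade omitted on forward and mobile.
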Material /BRCosmetics/Weapons/FORT_Melee/Pickaxe_FishBowl/FX/M_Fishbowl_MagicTrail
graph {
 properties {
  bUsedWithNiagaraMeshParticles=true
  bUsedWithNiagaraRibbons=true
  blendMode=BLEND_Additive
  pixelDepthOffsetMode=PDOM_Legacy
  shadingModel=MSM_Unlit
  twoSided=true
  usedShadingModels=Unlit
 }
 nodes [
  Constant() {r=1}
  n1 = TextureCoordinate() {uTiling=0.6, vTiling=0.5}
  n2 = Constant() {r=1}
  n3 = Multiply(A=TextureCoordinate(), B=Constant() {r=1})
  n4 = ComponentMask(None=TextureSample(UVs=Panner(Coordinate=@n3) {speedY=0.8}) {texture=/Game/Effects/Fort_Effects/Emotes/UltraLight/Texture/T_SoftNoise_BC7}.RGB) {g=true, r=true}
  n5 = Constant() {r=2}
  n6 = Panner(Coordinate=Add(A=Multiply(A=@n1, B=@n2), B=Divide(A=@n4, B=@n5))) {speedY=-1}
  n7 = MaterialFunctionCall!(
   Blend=TextureSample(UVs=@n6) {samplerType=SAMPLERTYPE_LinearGrayscale, texture=/Game/Effects/Fort_Effects/Emotes/UltraLight/Texture/T_PlasmaCoils}.R
   Base=TextureSample(UVs=Multiply(A=@n6) {constB=2}) {samplerType=SAMPLERTYPE_LinearGrayscale, texture=&T_PlasmaCoils}.R
  ) {materialFunction=/Engine/Functions/Engine_MaterialFunctions03/Blends/Blend_Lighten}
  n8 = Multiply(A=Constant() {r=20}, B=@n7.Result)
  n9 = Multiply(A=TextureCoordinate(), B=Constant() {r=1})
  n10 = Panner(Coordinate=Multiply(A=TextureCoordinate(), B=Constant() {r=1})) {speedY=1.2}
  n11 = Divide(
   A=ComponentMask(None=TextureSample(UVs=@n10) {samplerType=SAMPLERTYPE_LinearGrayscale, texture=/Game/Environments/Hexmap/HexClouds/T_HexmapFog_01_Height02}.RGB) {g=true, r=true}
   B=Constant() {r=2}
  )
  n12 = TextureSample(UVs=Panner(Coordinate=Add(A=@n9, B=@n11)) {speedY=1.5}) {texture=/Game/Effects/Fort_Effects/Textures/Tile/T_Noise_02}
  n13 = Constant() {r=5}
  n14 = Multiply(A=@n7.Result, B=Saturate(None=Multiply(A=@n12.RGB, B=@n13)))
  n15 = MaterialFunctionCall(LDR Input=@n14, HDR Dynamic Range=Constant() {r=10}, HDR Tint=Multiply(A=Constant() {r=200}, B=@n14)) {materialFunction=&DeriveHDRfromLDR}
  n16 = ScalarParameter() {defaultValue=1, parameterName=AorB}
  n17 = ParticleColor()
  MaterialFunctionCall(Emissive=Multiply(A=LinearInterpolate(A=@n8, B=@n15.RemappedRange, Alpha=@n16), B=@n17.RGB)) {materialFunction=&MF_FX_EmissiveAdjustments}
  n18 = Power(Base=@n14)
  n19 = MaterialFunctionCall(Float2=TextureCoordinate()) {materialFunction=&BreakOutFloat2Components}
  n20 = Saturate(None=Multiply(A=@n19.R) {constB=14})
  n21 = Multiply(A=OneMinus(None=@n19.R))
  n22 = Multiply(A=@n19.G) {constB=0.5}
  n23 = OneMinus(None=Multiply(A=Multiply(A=@n20, B=@n21), B=Sine(None=@n22)))
  n24 = DynamicParameter()
  n25 = Saturate(None=Subtract(A=LinearInterpolate(A=@n7.Result, B=@n18, Alpha=@n16), B=Subtract(A=@n23, B=@n24.Param1)))
  n26 = DepthFade(Opacity=@n25)
  Multiply(A=@n17.A, B=ShadingPathSwitch(Default=@n26, Deferred=@n26, Forward=@n25, Mobile=@n25))
 ]
}
