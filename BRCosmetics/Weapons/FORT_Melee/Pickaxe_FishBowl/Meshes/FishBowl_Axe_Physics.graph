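# Physics asset for the FishBowl pickaxe, giving its device body and dangling-named glass hinge bones capsule collision and joints.
# Cone, gear, round and small-gear hinge chains each attach to DeviceBody, with longer chains linking bone to bone.
# Cone and gear chains have all angular motion locked, as does the middle round joint.
# The small-gear joint and the round chain's root and tip joints leave angular motion unlocked.
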PhysicsAsset /BRCosmetics/Weapons/FORT_Melee/Pickaxe_FishBowl/Meshes/FishBowl_Axe_Physics
bodies {
 DeviceBody [{shapeType=Capsule, length=66.9599, radius=6.2359, center=<30.2768 -0.0341 -0.0435>, rotation=<rot 0 90 90>}]
 dyn_ConeGlassHinge_01 [{shapeType=Capsule, length=2.331, radius=0.75, center=<0.6529 0.0082 0>, rotation=<rot 0.7228 90.7176 -90>}]
 dyn_ConeGlassHinge_02 [{shapeType=Capsule, length=3.8705, radius=3.3421, center=<0.4475 0 0>, rotation=<rot -90 0 -0.0003>}]
 dyn_GearGlassHinge_01 [{shapeType=Capsule, length=1.7387, radius=1, rotation=<rot -0.5243 -91.3492 -90>}]
 dyn_GearGlassHinge_02 [{shapeType=Capsule, length=3.43, radius=4.75, center=<0.9443 0 0>, rotation=<rot 0 -90.0007 90>}]
 dyn_RoundGlassHinge_01 [{shapeType=Capsule, length=0, radius=1.25, shapeName=dyn_RoundGlassHinge_01_capsule}]
 dyn_RoundGlassHinge_02 [{shapeType=Capsule, length=1.0401, radius=2.6336, center=<0.1605 -0.202 -0.3008>, rotation=<rot -1.1722 -99.9339 -6.56>}]
 dyn_RoundGlassHinge_03 [{shapeType=Capsule, length=5.9807, radius=7.0449, center=<22.9548 -0.0001 0>, rotation=<rot -90 -90 89.9998>}]
 dyn_SmallGearGlassHinge_01 [{shapeType=Capsule, length=4.9541, radius=2.7941, center=<4.7963 -0.0152 2.216>, rotation=<rot -1.5754 89.0903 -90>}]
}
constraints [
 {
  bone1Name=dyn_GearGlassHinge_02
  bone2Name=dyn_GearGlassHinge_01
  swing1LimitDegrees=0
  swing1Motion=Locked
  swing2LimitDegrees=35
  swing2Motion=Locked
  twistLimitDegrees=0
  twistMotion=Locked
 }
 {bone1Name=dyn_SmallGearGlassHinge_01, bone2Name=DeviceBody, swing1LimitDegrees=0, swing2LimitDegrees=180, twistLimitDegrees=0}
 {
  bone1Name=dyn_ConeGlassHinge_01
  bone2Name=DeviceBody
  swing1LimitDegrees=66.24
  swing1Motion=Locked
  swing2LimitDegrees=70.56
  swing2Motion=Locked
  twistLimitDegrees=47.36
  twistMotion=Locked
 }
 {
  bone1Name=dyn_ConeGlassHinge_02
  bone2Name=dyn_ConeGlassHinge_01
  swing1LimitDegrees=0
  swing1Motion=Locked
  swing2LimitDegrees=5
  swing2Motion=Locked
  twistLimitDegrees=0
  twistMotion=Locked
 }
 {bone1Name=dyn_RoundGlassHinge_03, bone2Name=dyn_RoundGlassHinge_02, swing1LimitDegrees=92.16, swing2LimitDegrees=0, twistLimitDegrees=0}
 {
  bone1Name=dyn_GearGlassHinge_01
  bone2Name=DeviceBody
  swing1LimitDegrees=0
  swing1Motion=Locked
  swing2LimitDegrees=0
  swing2Motion=Locked
  twistLimitDegrees=180
  twistMotion=Locked
 }
 {
  bone1Name=dyn_RoundGlassHinge_02
  bone2Name=dyn_RoundGlassHinge_01
  swing1LimitDegrees=0
  swing1Motion=Locked
  swing2LimitDegrees=15
  swing2Motion=Locked
  twistLimitDegrees=0
  twistMotion=Locked
 }
 {bone1Name=dyn_RoundGlassHinge_01, bone2Name=DeviceBody, swing1LimitDegrees=0, swing2LimitDegrees=0, twistLimitDegrees=5}
]
