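# Animation Blueprint for the FishBowl pickaxe that adds secondary physics motion to its glass hinge bones.
# A cached input pose feeds a RigidBody node using a dedicated physics asset, and a fallback chain of one bone offset and four spring-driven AnimDynamics nodes.
# bRigidBodyEnabled selects RigidBody when true, AnimDynamics otherwise.
# The update event lowers only the RigidBody alpha in the front end.
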AnimBlueprint /BRCosmetics/Weapons/FORT_Melee/Pickaxe_FishBowl/Meshes/FishBowl_Axe_AnimBlueprint
target_skeleton=&/FishBowl_Axe_Skeleton
graphs [
 AnimGraph {nodes [
   n1 = RigidBody(ComponentPose=Local To Component(LocalPose=Use cached pose 'BasePose'()), Alpha=Get Alpha() -> (Alpha:float)) {
    bClampLinearTranslationLimitToRefPose=true
    bOverrideWorldGravity=true
    overridePhysicsAsset=&/FishBowl_Axe_Physics
    simulationSpace=BaseBoneSpace
    alphaBoolBlend {blendOption=Linear}
    baseBoneRef {boneName=DeviceBody}
    componentAppliedLinearAccClamp=<3000 3000 3000>
    componentLinearAccScale=<0.4 0.3 0.3>
    componentLinearVelScale=<0.4 0.3 0.3>
    overrideWorldGravity {z=-980}
   }
   n2 = "Transform (Modify) Bone - Bone: dyn_RoundGlassHinge_02"(ComponentPose=Local To Component(LocalPose=Use cached pose 'BasePose'()), Alpha=1, Translation="0,6,0", Rotation="0,0,0", Scale="1,1,1") {
    translationMode=BMM_Additive
    translationSpace=BCS_BoneSpace
    alphaBoolBlend {blendOption=Linear}
    boneToModify {boneName=dyn_RoundGlassHinge_02}
    translation {y=6}
   }
   n3 = Anim Dynamics - Bone: dyn_SmallGearGlassHinge_01(ComponentPose=@n2, Alpha=1, ExternalForce="0,0,0") {
    angularDampingOverride=1
    angularSpringConstant=4000
    bAngularSpring=true
    bOverrideAngularDamping=true
    bUseGravityOverride=true
    gravityScale=0.35
    simulationSpace=BoneRelative
    alphaBoolBlend {blendOption=Linear}
    boundBone {boneName=dyn_SmallGearGlassHinge_01}
    componentAppliedLinearAccClamp=<3000 3000 3000>
    componentLinearAccScale=<1 1 1>
    componentLinearVelScale=<1 1 1>
    gravityOverride {z=-980}
    physicsBodyDefinitions [{
      boundBone {boneName=dyn_SmallGearGlassHinge_01}
      boxExtents=<20 30 30>
      constraintSetup {angularLimitsMax=<0 -15 0>, angularLimitsMin=<0 -55 0>, angularTarget=<1 0 0>}
      localJointOffset=<4 0 0>
     }]
    relativeSpaceBone {boneName=DeviceBody}
   }
   n4 = Anim Dynamics - Bone: dyn_RoundGlassHinge_03(ComponentPose=@n3, Alpha=1, ExternalForce="0,0,0") {
    angularDampingOverride=0.99
    angularSpringConstant=2500
    bAngularSpring=true
    bOverrideAngularDamping=true
    bUseGravityOverride=true
    bUsePlanarLimit=false
    gravityScale=0.35
    simulationSpace=BoneRelative
    alphaBoolBlend {blendOption=Linear}
    boundBone {boneName=dyn_RoundGlassHinge_03}
    componentAppliedLinearAccClamp=<3000 3000 3000>
    componentLinearAccScale=<1.1 1.1 1.1>
    componentLinearVelScale=<1.2 1.3 1.2>
    gravityOverride {z=-980}
    physicsBodyDefinitions [{
      boundBone {boneName=dyn_RoundGlassHinge_03}
      boxExtents=<20 30 50>
      constraintSetup {angularLimitsMax=<0 0 95>, angularLimitsMin=<0 0 55>, angularTarget=<0 0 0>}
      localJointOffset=<5 0 0>
     }]
    planarLimits [{drivingBone {boneName=dyn_RoundGlassHinge_03}, planeTransform {location=<0 5 0>, rotation=<rot 0 0 -90>}}]
    relativeSpaceBone {boneName=b_AR_Root}
   }
   n5 = Anim Dynamics - Bone: dyn_RoundGlassHinge_02(ComponentPose=@n4, Alpha=1, ExternalForce="0,0,0") {
    angularDampingOverride=1
    angularSpringConstant=4000
    bAngularSpring=true
    bGravityOverrideInSimSpace=true
    bLinearSpring=true
    bOverrideAngularDamping=true
    bOverrideLinearDamping=true
    bUseGravityOverride=true
    gravityScale=0.35
    linearDampingOverride=1
    linearSpringConstant=100
    simulationSpace=BoneRelative
    alphaBoolBlend {blendOption=Linear}
    boundBone {boneName=dyn_RoundGlassHinge_02}
    componentAppliedLinearAccClamp=<3000 3000 3000>
    componentLinearAccScale=<0.5 0.5 0.5>
    componentLinearVelScale=<0.5 0.5 0.5>
    gravityOverride {z=-980}
    physicsBodyDefinitions [{
      boundBone {boneName=dyn_RoundGlassHinge_02}
      boxExtents=<20 30 30>
      constraintSetup {angularLimitsMax=<0 0 0>, angularLimitsMin=<0 0 0>, angularTarget=<1 0 0>}
      localJointOffset=<0 5 0>
     }]
    relativeSpaceBone {boneName=Melee_TwoHanded_Handle}
   }
   n6 = Anim Dynamics - Bone: dyn_RoundGlassHinge_01(ComponentPose=@n5, Alpha=1, ExternalForce="0,0,0") {
    angularDampingOverride=1
    angularSpringConstant=4000
    bAngularSpring=true
    bOverrideAngularDamping=true
    bUseGravityOverride=true
    gravityScale=0.35
    simulationSpace=BoneRelative
    alphaBoolBlend {blendOption=Linear}
    boundBone {boneName=dyn_RoundGlassHinge_01}
    componentAppliedLinearAccClamp=<3000 3000 3000>
    componentLinearAccScale=<1 1 1>
    componentLinearVelScale=<1 1 1>
    gravityOverride {z=-980}
    physicsBodyDefinitions [{
      boundBone {boneName=dyn_RoundGlassHinge_01}
      boxExtents=<20 30 30>
      constraintSetup {angularLimitsMax=<0 0 0>, angularLimitsMin=<0 0 0>, angularTarget=<1 0 0>}
      localJointOffset=<4 0 0>
     }]
    relativeSpaceBone {boneName=DeviceBody}
   }
   Output Pose(Result=Blend Poses by bool(BlendPose_0=Component To Local(ComponentPose=@n1), BlendPose_1=Component To Local(ComponentPose=@n6), bActiveValue=Get bRigidBodyEnabled()) {blendType=Linear}) {layerGroup=None}
   Save cached pose 'BasePose'(Pose=Input Pose()) {cacheName=BasePose}
  ]}
 EventGraph {nodes [
   Event BlueprintUpdateAnimation()
   > n1 = Branch(Condition=Get bIsInFrontEnd())
   > Set Alpha(Alpha:float=0.8) -> (Output_Get:float)
   Set Alpha(execute=@n1.else, Alpha=1)
   TryGetPawnOwner()
  ]}
]
variables [Alpha {type=number, default=1}]
